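# Physics asset for the Cactus Rocker male medium player cosmetic body, covering a dangling thigh necklace on the left leg.
# Capsule bodies cover two dynamic necklace bones and the left thigh twist bone, whose larger capsule spans the thigh.
# Angular-limited constraints chain the second necklace bone to the first and the first to the thigh, with tighter swing at the thigh joint and tighter twist between the necklace links.
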PhysicsAsset /BRCosmetics/Characters/Player/Male/Medium/Bodies/M_MED_Cactus_Rocker/Meshes/M_MED_Cactus_Rocker_Physics
bodies {
 dyn_thigh_necklace_01 [{
   shapeType=Capsule
   length=5.3743
   radius=2.512
   shapeName=dyn_thigh_necklace_01_capsule
   center=<-1.7501 -2.1449 -0.9936>
   rotation=<rot -10.6499 -85.2352 -4.0335>
  }]
 dyn_thigh_necklace_02 [{
   shapeType=Capsule
   length=5.0234
   radius=2.7029
   shapeName=dyn_thigh_necklace_02_capsule
   center=<-1.6916 -0.9465 -1.2312>
   rotation=<rot -6.2341 -84.6567 179.2588>
  }]
 thigh_twist_01_l [{shapeType=Capsule, length=43.7522, radius=7.0475, center=<-0.1453 -0.2258 0.1404>, rotation=<rot 0.7907 -89.6392 -94.0817>}]
}
constraints [
 {bone1Name=dyn_thigh_necklace_01, bone2Name=thigh_twist_01_l, swing1LimitDegrees=30, swing2LimitDegrees=10, twistLimitDegrees=15}
 {bone1Name=dyn_thigh_necklace_02, bone2Name=dyn_thigh_necklace_01, swing1LimitDegrees=45, swing2LimitDegrees=45, twistLimitDegrees=5}
]
