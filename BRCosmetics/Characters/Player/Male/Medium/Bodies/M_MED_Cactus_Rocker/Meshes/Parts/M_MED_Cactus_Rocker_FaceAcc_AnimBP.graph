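# Animation Blueprint for the Cactus Rocker male medium character's face accessory, adding secondary jiggle to the head strap and two hair bones.
# It copies the pose from the attached parent mesh, then runs three chained Anim Dynamics nodes on dyn_HeadStrap, dyn_hair_01 and dyn_hair_02.
# Each is a partially blended, angular-spring body with tight angle limits, reduced gravity and heavy damping, simulated root-relative.
# The event graph holds no logic.
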AnimBlueprint /BRCosmetics/Characters/Player/Male/Medium/Bodies/M_MED_Cactus_Rocker/Meshes/Parts/M_MED_Cactus_Rocker_FaceAcc_AnimBP
target_skeleton=&/M_MED_Cactus_Rocker_FaceAcc_Skeleton
graphs [
 AnimGraph {nodes [
   n1 = Anim Dynamics - Bone: dyn_HeadStrap(ComponentPose=Local To Component(LocalPose=Copy Pose From Mesh() {bUseAttachedParent=true}), Alpha=0.7, ExternalForce="0,0,0") {
    angularDampingOverride=0.999
    angularSpringConstant=800
    bAngularSpring=true
    bGravityOverrideInSimSpace=true
    bOverrideAngularDamping=true
    gravityScale=0.2
    simulationSpace=RootRelative
    alphaBoolBlend {blendOption=Linear}
    boundBone {boneName=dyn_HeadStrap}
    componentAppliedLinearAccClamp=<10000 10000 10000>
    componentLinearAccScale=<0.1 0.1 0.1>
    componentLinearVelScale=<0.1 0.1 0.1>
    physicsBodyDefinitions [{
      sphereCollisionRadius=0
      boundBone {boneName=dyn_HeadStrap}
      boxExtents=<5 8 8>
      constraintSetup {twistAxis=AxisY, angularLimitsMax=<15 10 2>, angularLimitsMin=<-15 -10 -2>, angularTarget=<0 0 0>}
      localJointOffset=<-5 0.1 1>
     }]
    relativeSpaceBone {boneName=pelvis}
   }
   n2 = Anim Dynamics - Bone: dyn_hair_01(ComponentPose=@n1, Alpha=0.7) {
    angularDampingOverride=0.999
    angularSpringConstant=800
    bAngularSpring=true
    bGravityOverrideInSimSpace=true
    bOverrideAngularDamping=true
    gravityScale=0.2
    simulationSpace=RootRelative
    alphaBoolBlend {blendOption=Linear}
    boundBone {boneName=dyn_hair_01}
    componentAppliedLinearAccClamp=<10000 10000 10000>
    componentLinearAccScale=<0.1 0.1 0.1>
    componentLinearVelScale=<0.1 0.1 0.1>
    physicsBodyDefinitions [{
      sphereCollisionRadius=0
      boundBone {boneName=dyn_hair_01}
      boxExtents=<5 8 8>
      constraintSetup {angularTargetAxis=AxisY, angularLimitsMax=<5 10 5>, angularLimitsMin=<-5 -10 5>, angularTarget=<0 0 0>}
      localJointOffset=<-5 0.1 0.5>
     }]
    relativeSpaceBone {boneName=pelvis}
   }
   n3 = Anim Dynamics - Bone: dyn_hair_02(ComponentPose=@n2, Alpha=0.7) {
    angularDampingOverride=0.999
    angularSpringConstant=800
    bAngularSpring=true
    bGravityOverrideInSimSpace=true
    bOverrideAngularDamping=true
    gravityScale=0.2
    simulationSpace=RootRelative
    alphaBoolBlend {blendOption=Linear}
    boundBone {boneName=dyn_hair_02}
    componentAppliedLinearAccClamp=<10000 10000 10000>
    componentLinearAccScale=<0.1 0.1 0.1>
    componentLinearVelScale=<0.1 0.1 0.1>
    physicsBodyDefinitions [{
      sphereCollisionRadius=0
      boundBone {boneName=dyn_hair_02}
      boxExtents=<5 8 8>
      constraintSetup {angularTargetAxis=AxisY, angularLimitsMax=<5 10 5>, angularLimitsMin=<-5 -10 5>, angularTarget=<0 0 0>}
      localJointOffset=<-5 0.1 0.5>
     }]
    relativeSpaceBone {boneName=pelvis}
   }
   Output Pose(Result=Component To Local(ComponentPose=@n3)) {layerGroup=None}
  ]}
 EventGraph {nodes [Event BlueprintUpdateAnimation()]}
]
variables []
